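# Physics asset for the face accessory mesh of the large female Blonde character's Cat skin, covering left and right cat ears, loops and front and back ribbons.
# Each piece is a two-bone chain of box bodies constrained to a head body built from three capsules.
# Swing and twist limits keep the ears nearly rigid, especially at the root, while loops and ribbons swing more freely with tight twist.
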PhysicsAsset /BRCosmetics/Characters/Player/Female/Large/Bodies/F_LRG_Blonde/Skins/Cat/Meshes/F_LRG_Blonde_Cat_FaceAcc_Physics
bodies {
 dyn_l_ear_01 [{shapeType=Box, extentX=3.0178, extentY=4.291, extentZ=2, center=<1.7513 0 0.0001>}]
 dyn_l_ear_02 [{shapeType=Box, extentX=3.6348, extentY=3.9633, extentZ=2, center=<1.8574 0 -0.0002>}]
 dyn_l_loop_01 [{shapeType=Box, extentX=1.4805, extentY=2, extentZ=1.5084, center=<0.9733 0.0001 0>}]
 dyn_l_loop_02 [{shapeType=Box, extentX=2, extentY=2, extentZ=1.2929, center=<0.9878 0.0001 0>}]
 dyn_l_ribbon_b_01 [{shapeType=Box, extentX=2.3632, extentY=1.2501, extentZ=0.874, center=<1.3945 0 0>}]
 dyn_l_ribbon_b_02 [{shapeType=Box, extentX=2, extentY=0.9995, extentZ=1.1879, center=<1.1954 0 -0.0001>}]
 dyn_l_ribbon_f_01 [{shapeType=Box, extentX=2.6498, extentY=1.3972, extentZ=0.9304, center=<1.6228 0 -0.0001>}]
 dyn_l_ribbon_f_02 [{shapeType=Box, extentX=2.7381, extentY=1.2871, extentZ=1.1879, center=<1.4423 0 -0.0002>}]
 dyn_r_ear_01 [{shapeType=Box, extentX=3.0178, extentY=4.291, extentZ=2, center=<1.7513 0 0.0001>}]
 dyn_r_ear_02 [{shapeType=Box, extentX=3.6348, extentY=3.9633, extentZ=2, center=<1.8574 0 -0.0002>}]
 dyn_r_loop_01 [{shapeType=Box, extentX=1.4805, extentY=2, extentZ=1.6728, center=<0.9733 0.0001 0>}]
 dyn_r_loop_02 [{shapeType=Box, extentX=2, extentY=2, extentZ=1.6146, center=<0.9878 0.0001 0>}]
 dyn_r_ribbon_b_01 [{shapeType=Box, extentX=2.3632, extentY=1.2501, extentZ=0.7826, center=<1.3945 0 0>}]
 dyn_r_ribbon_b_02 [{shapeType=Box, extentX=2, extentY=0.9995, extentZ=0.8261, center=<1.1954 0 -0.0001>}]
 dyn_r_ribbon_f_01 [{shapeType=Box, extentX=2.6498, extentY=1.3972, extentZ=0.7826, center=<1.6228 0 -0.0001>}]
 dyn_r_ribbon_f_02 [{shapeType=Box, extentX=2.7381, extentY=1.2871, extentZ=1.0167, center=<1.4423 0 -0.0002>}]
 head [
  {shapeType=Capsule, length=12.9594, radius=8.2532, rotation=<rot 90 0 0>}
  {shapeType=Capsule, length=6.8104, radius=4.7637, center=<6.7951 2.3355 4.5186>, rotation=<rot 53.8052 16.3397 -75.3379>}
  {shapeType=Capsule, length=3.4996, radius=4.7637, center=<5.0323 1.1152 -3.9994>, rotation=<rot 84.4211 -179.9996 -91.5956>}
 ]
}
constraints [
 {bone1Name=dyn_l_ribbon_f_02, bone2Name=dyn_l_ribbon_f_01, swing1LimitDegrees=30, swing2LimitDegrees=30, twistLimitDegrees=5}
 {bone1Name=dyn_l_ribbon_b_02, bone2Name=dyn_l_ribbon_b_01, swing1LimitDegrees=30, swing2LimitDegrees=30, twistLimitDegrees=5}
 {bone1Name=dyn_l_loop_02, bone2Name=dyn_l_loop_01, swing1LimitDegrees=30, swing2LimitDegrees=30, twistLimitDegrees=5}
 {bone1Name=dyn_l_ear_02, bone2Name=dyn_l_ear_01, swing1LimitDegrees=10, swing2LimitDegrees=10, twistLimitDegrees=5}
 {bone1Name=dyn_r_ear_02, bone2Name=dyn_r_ear_01, swing1LimitDegrees=10, swing2LimitDegrees=10, twistLimitDegrees=5}
 {bone1Name=dyn_r_loop_02, bone2Name=dyn_r_loop_01, swing1LimitDegrees=30, swing2LimitDegrees=30, twistLimitDegrees=5}
 {bone1Name=dyn_r_ribbon_f_02, bone2Name=dyn_r_ribbon_f_01, swing1LimitDegrees=30, swing2LimitDegrees=30, twistLimitDegrees=5}
 {bone1Name=dyn_r_ribbon_b_02, bone2Name=dyn_r_ribbon_b_01, swing1LimitDegrees=30, swing2LimitDegrees=30, twistLimitDegrees=5}
 {bone1Name=dyn_l_ribbon_b_01, bone2Name=head, swing1LimitDegrees=30, swing2LimitDegrees=30, twistLimitDegrees=5}
 {bone1Name=dyn_l_ribbon_f_01, bone2Name=head, swing1LimitDegrees=30, swing2LimitDegrees=30, twistLimitDegrees=5}
 {bone1Name=dyn_l_loop_01, bone2Name=head, swing1LimitDegrees=30, swing2LimitDegrees=30, twistLimitDegrees=5}
 {bone1Name=dyn_l_ear_01, bone2Name=head, swing1LimitDegrees=0, swing2LimitDegrees=5, twistLimitDegrees=5}
 {bone1Name=dyn_r_ear_01, bone2Name=head, swing1LimitDegrees=0, swing2LimitDegrees=5, twistLimitDegrees=5}
 {bone1Name=dyn_r_loop_01, bone2Name=head, swing1LimitDegrees=30, swing2LimitDegrees=30, twistLimitDegrees=5}
 {bone1Name=dyn_r_ribbon_f_01, bone2Name=head, swing1LimitDegrees=30, swing2LimitDegrees=30, twistLimitDegrees=5}
 {bone1Name=dyn_r_ribbon_b_01, bone2Name=head, swing1LimitDegrees=30, swing2LimitDegrees=30, twistLimitDegrees=5}
]
