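# Animation Blueprint for the face accessory on the large female Blonde character's Cat skin, adding physics-driven secondary motion.
# It copies the attached parent's pose into a RigidBody node that simulates in head-bone space with an override physics asset, custom gravity and scaled component motion.
# A softer RigidBody branch feeding a bIsTargeting blend is not connected to the output pose.
# The event graph contains only unconnected stub nodes.
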AnimBlueprint /BRCosmetics/Characters/Player/Female/Large/Bodies/F_LRG_Blonde/Skins/Cat/Meshes/F_LRG_Blonde_Cat_FaceAcc_AnimBP
target_skeleton=&/F_LRG_Blonde_Cat_FaceAcc_Skeleton
graphs [
 AnimGraph {nodes [
   n1 = RigidBody(ComponentPose=Local To Component(LocalPose=Copy Pose From Mesh() {bUseAttachedParent=true}), Alpha=0.8) {
    bClampLinearTranslationLimitToRefPose=true
    bOverrideWorldGravity=true
    overridePhysicsAsset=&/F_LRG_Blonde_Cat_FaceAcc_Physics
    simulationSpace=BaseBoneSpace
    alphaBoolBlend {blendOption=Linear}
    baseBoneRef {boneName=head}
    componentAppliedLinearAccClamp=<5000 5000 5000>
    componentLinearAccScale=<0.3 0.3 0.4>
    componentLinearVelScale=<0.15 0.15 0.2>
    overrideWorldGravity {z=-980}
    simSpaceSettings {
     maxAngularAcceleration=500
     maxAngularVelocity=3
     maxLinearAcceleration=400
     maxLinearVelocity=250
     velocityScaleZ=0
     worldAlpha=1
     externalLinearDragV=<0.1 0.1>
    }
   }
   Blend Poses by bool(BlendPose_0=Component To Local(ComponentPose=@n1), bActiveValue=Get bIsTargeting()) {blendType=Linear}
   n2 = RigidBody(ComponentPose=Local To Component(LocalPose=Copy Pose From Mesh() {bUseAttachedParent=true})) {
    bClampLinearTranslationLimitToRefPose=true
    bOverrideWorldGravity=true
    overridePhysicsAsset=&F_LRG_Blonde_Cat_FaceAcc_Physics
    simulationSpace=BaseBoneSpace
    alphaBoolBlend {blendOption=Linear}
    baseBoneRef {boneName=head}
    componentAppliedLinearAccClamp=<5000 5000 5000>
    componentLinearAccScale=<0.6 0.6 0.7>
    componentLinearVelScale=<0.3 0.3 0.4>
    overrideWorldGravity {z=-980}
    simSpaceSettings {
     maxAngularAcceleration=500
     maxAngularVelocity=3
     maxLinearAcceleration=400
     maxLinearVelocity=250
     velocityScaleZ=0
     worldAlpha=1
     externalLinearDragV=<0.1 0.1>
    }
   }
   Output Pose(Result=Component To Local(ComponentPose=@n2)) {layerGroup=None}
  ]}
 EventGraph {nodes [
   TryGetPawnOwner()
   Event BlueprintUpdateAnimation()
  ]}
]
variables []
